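# Deferred decal master material for Megascans (Bridge presets), translucent and Default Lit, built on virtual texture samplers.
# Tiled, crop-offset UVs optionally pass through Quixel parallax occlusion mapping via the UsePOM switch, then sample Albedo, Normal and DRO.
# Packed DRO supplies height in red, roughness in green and opacity in blue.
# Parameters adjust overlay tint, normal flatness and channel intensities; a switch picks metallic map or scalar, the map ignoring those UVs.
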Material /Bridge/MSPresets/M_MS_Decal_Material_VT/M_MS_Decal_Material_VT
graph {
 properties {blendMode=BLEND_Translucent, materialDomain=MD_DeferredDecal, pixelDepthOffsetMode=PDOM_Legacy, usedShadingModels=Default Lit}
 nodes [
  Constant() {r=0.5}
  n1 = TextureObjectParameter() {
   group=01 - Input Textures
   parameterName=DRO
   samplerType=SAMPLERTYPE_VirtualLinearColor
   sortPriority=3
   texture=/Bridge/MSPresets/MSVTTextures/WhitePlaceholder
  }
  n2 = ScalarParameter() {defaultValue=1, group=02- UV Controls, parameterName=Crop X, sortPriority=1}
  n3 = ScalarParameter() {defaultValue=1, group=02- UV Controls, parameterName=Crop Y, sortPriority=2}
  n4 = Add(
   A=Multiply(A=TextureCoordinate(), B=ScalarParameter() {defaultValue=1, group=02- UV Controls, parameterName=Tiling})
   B=Add(
    A=Constant() {r=1}
    B=MaterialFunctionCall(X=@n2, Y=@n3) {materialFunction=&MakeFloat2}.Result
   )
  )
  n5 = StaticSwitchParameter(
   True=MaterialFunctionCall!(
    Heightmap Texture=@n1
    Height Ratio=ScalarParameter() {defaultValue=0.05, group=08 - Parallax, parameterName=Height Ratio, sortPriority=1}
    Min Steps=Constant() {r=8}
    Max Steps=Constant() {r=64}
    UVs=@n4
    Heightmap Channel=Constant4Vector() {constant=<rgba 1 0 0 0>}.RGBA
    Reference Plane=ScalarParameter() {defaultValue=0.72, group=08 - Parallax, parameterName=Reference Plane, sortPriority=2}
    -----------------------------
    Use World Coordinates
    Specify Manual Texture Size
    Manual Texture Size
    ------------------------------
    "Render Shadows (Occlusion Mapping)"
    Light Vector
    Shadow Steps
    Shadow Penumbra
    Transform To VertexNormal
   ) {materialFunction=/Bridge/MSPresets/M_MS_Decal_Material/MF_QuixelDecalPOM}.Parallax UVs
   False=@n4
  ) {group=08 - Parallax, parameterName=UsePOM}
  n6 = TextureSampleParameter2D(UVs=@n5) {
   group=01 - Input Textures
   parameterName=Albedo
   samplerSource=SSM_Wrap_WorldGroupSettings
   samplerType=SAMPLERTYPE_VirtualColor
   sortPriority=0
   texture=/Bridge/MSPresets/MSVTTextures/DefaultDiffuse
  }
  LinearInterpolate(
   A=@n6.RGB
   B=MaterialFunctionCall(Base=@n6.RGB, Blend=VectorParameter() {group=03 - Albedo, parameterName=Color Overlay, defaultValue=<rgba 0.5 0.5 0.5 1>}.RGB) {materialFunction=&Blend_Overlay}.Result
   Alpha=ScalarParameter() {defaultValue=1, group=03 - Albedo, parameterName=Overlay Intensity, sliderMax=1.5, sortPriority=1}
  ) {desc=ms_BaseColor}
  n7 = TextureSampleParameter2D(UVs=@n5) {
   group=01 - Input Textures
   parameterName=Normal
   samplerSource=SSM_Wrap_WorldGroupSettings
   samplerType=SAMPLERTYPE_VirtualNormal
   sortPriority=2
   texture=/Bridge/MSPresets/MSVTTextures/Placeholder_Normal
  }
  MaterialFunctionCall(Normal=@n7.RGB, Flatness=OneMinus(None=ScalarParameter() {defaultValue=1, group=07 - Normal, parameterName=Normal Intensity})) {materialFunction=&FlattenNormal}
  n8 = TextureSample(UVs=@n5, Tex=@n1)
  Multiply(A=@n8.G, B=ScalarParameter() {defaultValue=1, group=05 - Roughness, parameterName=Roughness Intensity}) {desc=ms_Roughness}
  Multiply(A=@n8.B, B=ScalarParameter() {defaultValue=1, group=06 - Opacity, parameterName=Opacity Intensity}) {desc=ms_Opacity}
  n9 = TextureSampleParameter2D() {
   group=01 - Input Textures
   parameterName=Metallic
   samplerSource=SSM_Wrap_WorldGroupSettings
   samplerType=SAMPLERTYPE_VirtualLinearColor
   sortPriority=1
   texture=/Bridge/MSPresets/MSVTTextures/BlackPlaceholder
  }
  StaticSwitchParameter(
   True=Multiply(A=@n9.R, B=ScalarParameter() {defaultValue=1, group=04 - Metallic, parameterName=Metallic Map Intensity, sortPriority=2}) {desc=ms_Metallic}
   False=ScalarParameter() {group=04 - Metallic, parameterName=Base Metallic, sortPriority=1}
  ) {group=04 - Metallic, parameterName=Metallic Map}
 ]
}
